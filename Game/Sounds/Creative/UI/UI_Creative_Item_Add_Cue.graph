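# Sound cue for the Creative mode UI sound that, judging by its name, plays when an item is added.
# A single wave player feeds a Modulator that randomizes playback slightly, with the pitch floor raised to unshifted.
# It routes through the Creative gadgets sound class at reduced volume and borrows the Move Tool gadget's attenuation settings.
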SoundCue /Game/Sounds/Creative/UI/UI_Creative_Item_Add_Cue
root=Modulator_1
nodes [
 {id=Modulator_1, children [WavePlayer_0], properties {pitchMin=1}}
 {id=WavePlayer_0, properties {soundWaveAssetPtr=/Game/Sounds/Creative/UI/wav/UI_Creative_Item_Add_01}}
]
properties {
 attenuationSettings=/Game/Sounds/Creative/Attenuations/Fort_Gadget_MoveTool_Att
 maxDistance=4000
 soundClassObject=&Fort_Gadgets_Creative
 volumeMultiplier=0.6
}
