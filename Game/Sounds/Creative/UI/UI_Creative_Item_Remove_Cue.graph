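# Sound cue for the Creative mode UI sound played on item removal, in the Creative gadgets sound class.
# A single wave player feeds a Modulator left at default settings, with the cue's volume reduced.
# It uses the move tool's attenuation asset.
# A concurrency override limits it to one instance per owner, blocking new plays rather than cutting off the current one, with a short retrigger time.
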SoundCue /Game/Sounds/Creative/UI/UI_Creative_Item_Remove_Cue
root=Modulator_1
nodes [
 {id=Modulator_1, children [WavePlayer_1]}
 {id=WavePlayer_1, properties {soundWaveAssetPtr=/Game/Sounds/Creative/UI/wav/UI_Creative_Item_Remove_01}}
]
properties {
 attenuationSettings=/Game/Sounds/Creative/Attenuations/Fort_Gadget_MoveTool_Att
 bOverrideConcurrency=true
 maxDistance=4000
 soundClassObject=&Fort_Gadgets_Creative
 volumeMultiplier=0.6
 concurrencyOverrides {bLimitToOwner=true, maxCount=1, resolutionRule=PreventNew, retriggerTime=0.2}
}
